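# Test unlit weapon-skin material, named as a top projection, that maps a centerpoint debug texture across skeletal weapon meshes.
# The Y and Z components of pixel-shader weapon local position are divided by object radius and offset by 0.5 to form planar UVs.
# These UVs sample the debug texture as linear color, though the graph shows no connection from the sample to a material output.
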Material /Game/Weapons/WeaponSkins/Tests/8_TopProjection
graph {
 properties {bUsedWithMeshDeformer=true, bUsedWithSkeletalMesh=true, pixelDepthOffsetMode=PDOM_Legacy, shadingModel=MSM_Unlit, usedShadingModels=Unlit}
 nodes [
  n1 = ComponentMask(None=MaterialFunctionCall() {materialFunction=/Game/Packages/Fortnite_UniqueMaterials_1/Weapons/WeaponLocalPosition}.Weapon Local Position Pixel Shader) {b=true, g=true}
  n2 = Multiply(A=ObjectRadius())
  TextureSample(UVs=Add(A=Divide(A=@n1, B=@n2)) {constB=0.5}) {samplerType=SAMPLERTYPE_LinearColor, texture=/Game/Global/Textures/Default/Debugs/CenterpointDebugTexture}
 ]
}
